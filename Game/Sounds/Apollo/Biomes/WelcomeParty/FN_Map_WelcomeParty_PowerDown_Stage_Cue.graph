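# Sound cue that plays the power-down sound for the WelcomeParty map stage in the Apollo biome.
# A single wave player feeds the output with the FN_Map_WelcomeParty_PowerDown wave.
# It uses the stage main music attenuation asset with override enabled and is routed to the POI music sound class.
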SoundCue /Game/Sounds/Apollo/Biomes/WelcomeParty/FN_Map_WelcomeParty_PowerDown_Stage_Cue
root=WavePlayer_0
nodes [{id=WavePlayer_0, properties {soundWaveAssetPtr=&/FN_Map_WelcomeParty_PowerDown}}]
properties {
 attenuationSettings=&/WelcomeParty_Music_Stage_Main_Att
 bOverrideAttenuation=true
 maxDistance=4000
 soundClassObject=/Game/Sounds/FortSoundClassesAndModes/Fort_Music_POI_PSM
}
