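; Welcome Party biome audio controller that switches between day and night music for the local player within its overlap sphere.
; Server day-phase changes set a replicated struct of music state and start time, and clients fade out the old track before playing the new one.
; Entering the sphere, sized from the attenuation asset, seeks to server-elapsed time modulo track length for synced playback, and registers the submix for audio analysis.
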(event EventBeginPlay)

(event Custom|On%20Day%20Phase%20Changed%20Server
  (params (CurrentDayPhase :type EFortDayPhase%20Enum :default "Morning") (PreviousDayPhase :type EFortDayPhase%20Enum :default "Morning") (bAtCreation :type Boolean))
  (bind _returnvalue (Fortnite|Sync|Time|GetServerWorldTimeSeconds))
  (bind _musicstateserver (Variables|Default|GetMusicStateServer))
  (switch Utilities|FlowControl|Switch|SwitchonEFortDayPhase CurrentDayPhase
    (:Morning
      _)
    (:Day
      _)
    (:Evening
      _)
    (:Night
      (bind _updated_variable (Utilities|FlowControl|DeltaGateWildcard _musicstateserver 2)
        (:Success
          (Variables|Default|SetRepStruct (Utilities|Struct|MakeWelcomePartyAudioRepStruct _updated_variable _returnvalue bAtCreation))))))
  (bind _updated_variable_2 (Utilities|FlowControl|DeltaGateWildcard _musicstateserver 1)
    (:Success
      (Variables|Default|SetRepStruct (Utilities|Struct|MakeWelcomePartyAudioRepStruct _updated_variable_2 _returnvalue bAtCreation)))))

(event Custom|On%20Rep%20Struct%20Changed
  (params (RepStruct_MusicState_2_B1F205EA4007D7709EBA1EA327A0E085 :type Byte) (RepStruct_StartTime_5_0CD31FF5400675F1662677AA12CB505E :type Float%20%28single-precision%29) (RepStruct_AtCreation_8_1842B1AC41C4A9C988D15F865191FE13 :type Boolean))
  (bind _repstruct_musicstate_2_b1f205ea4007d7709eba1ea327a0e085 (Utilities|Struct|BreakWelcomePartyAudioRepStruct (Variables|GetRepStruct_MusicState_2_B1F205EA4007D7709EBA1EA327A0E085)))
  (Development|PrintString (Utilities|String|Append "WLCM MusicState: " (Utilities|String|ToString(Byte) RepStruct_MusicState_2_B1F205EA4007D7709EBA1EA327A0E085)))
  (if (Variables|Default|GetbLocalPlayerOverlapping)
    (switch int (Math|Conversions|ToInteger(Byte) RepStruct_MusicState_2_B1F205EA4007D7709EBA1EA327A0E085)
      (:1
        (Development|PrintString "WLCM After Hours: Disabled")
        (Audio|Components|Audio|FadeOut (Variables|Default|GetMusicNight) 3.0 0.0 "Linear" :!class /Script/Engine.AudioComponent)
        (Utilities|FlowControl|Delay 3.0)
        (if (Math|Byte|Equal(Byte) _repstruct_musicstate_2_b1f205ea4007d7709eba1ea327a0e085 1)
          (Audio|Components|Audio|Play (Variables|Default|GetMusicDay) 0.0)))
      (:2
        (Development|PrintString "WLCM After Hours: Enabled")
        (Audio|Components|Audio|FadeOut (Variables|Default|GetMusicDay) 3.0 0.0 "Linear" :!class /Script/Engine.AudioComponent)
        (Utilities|FlowControl|Delay 3.0)
        (if (Math|Byte|Equal(Byte) _repstruct_musicstate_2_b1f205ea4007d7709eba1ea327a0e085 2)
          (Audio|Components|Audio|Play (Variables|Default|GetMusicNight) 0.0))))
    (AudioAnalysis|RegisterDefaultSubmix (WorldSubsystems|GetFortAudioAnalysisSubsystem))))

(event OnComponentBeginOverlap(OverlapSphere)
  (params (OverlappedComponent :type Primitive%20Component%20Object%20Reference) (OtherActor :type Actor%20Object%20Reference) (OtherComp :type Primitive%20Component%20Object%20Reference) (OtherBodyIndex :type Integer) (bFromSweep :type Boolean) (SweepResult :type Hit%20Result%20Structure%20%28by%20ref%29))
  (CallFunction|OnOverlapStateChanged true OtherActor))

(event OnComponentEndOverlap(OverlapSphere)
  (params (OverlappedComponent :type Primitive%20Component%20Object%20Reference) (OtherActor :type Actor%20Object%20Reference) (OtherComp :type Primitive%20Component%20Object%20Reference) (OtherBodyIndex :type Integer))
  (CallFunction|OnOverlapStateChanged false OtherActor))

(event Custom|SetRandomTime
  (bind _musicday (Variables|Default|GetMusicDay))
  (bind _musicnight (Variables|Default|GetMusicNight))
  (if (Audio|Components|Audio|IsPlaying _musicnight)
    (Audio|Components|Audio|Play _musicnight (Math|Random|RandomFloatinRange 0.0 (Variables|/Script/Engine.SoundBase|GetDuration (Variables|Default|GetSoundWaveNight))))
    (elif (Audio|Components|Audio|IsPlaying _musicday)
      (Audio|Components|Audio|Play _musicday (Math|Random|RandomFloatinRange 0.0 (Variables|/Script/Engine.SoundBase|GetDuration (Variables|Default|GetSoundWaveDay)))))))

(fn OnOverlapStateChanged
  (params (Overlapping :type Boolean) (Actor :type Actor%20Object%20Reference))
  (bind _returnvalue (Fortnite|Sync|Time|GetServerWorldTimeSeconds))
  (bind (_repstruct_musicstate_2_b1f205ea4007d7709eba1ea327a0e085 _repstruct_starttime_5_0cd31ff5400675f1662677aa12cb505e) (Utilities|Struct|BreakWelcomePartyAudioRepStruct (Variables|GetRepStruct_MusicState_2_B1F205EA4007D7709EBA1EA327A0E085)))
  (bind _returnvalue_2 (- _returnvalue _repstruct_starttime_5_0cd31ff5400675f1662677aa12cb505e))
  (bind _soundwavenight (Variables|Default|GetSoundWaveNight))
  (bind _soundwaveday (Variables|Default|GetSoundWaveDay))
  (bind _returnvalue_3 (Utilities|Select :Index _repstruct_musicstate_2_b1f205ea4007d7709eba1ea327a0e085 :Option%201 _soundwaveday :Option%202 _soundwavenight))
  (bind _duration (Variables|/Script/Engine.SoundBase|GetDuration _returnvalue_3))
  (bind _returnvalue_4 (Math|Float|%25%28Float%29 _returnvalue_2 _duration))
  (bind _musicnight (Variables|Default|GetMusicNight))
  (bind _musicday (Variables|Default|GetMusicDay))
  (bind _returnvalue_5 (Utilities|Select :Index _repstruct_musicstate_2_b1f205ea4007d7709eba1ea327a0e085 :Option%201 _musicday :Option%202 _musicnight))
  (Networking|CanExecuteCosmeticEvents
    (:True
      (Variables|Default|SetbLocalPlayerOverlapping Overlapping)
      (if (Variables|Default|GetbMusicEnabled)
        (bind _asfort_pawn (Utilities|Casting|CastToFortPawn Actor))
        (if (Pawn|IsLocallyViewed _asfort_pawn)
          (if Overlapping
            (Utilities|IsValid _returnvalue_5
              (:Is%20Valid
                (Audio|Components|Audio|Play _returnvalue_5 _returnvalue_4)
                (Development|PrintString (Utilities|String|Append "WLCM Seek: " (Utilities|String|ToString(Float) _returnvalue_4)))
                (AudioAnalysis|RegisterDefaultSubmix (WorldSubsystems|GetFortAudioAnalysisSubsystem))))
            (elif (and (Utilities|IsValid _returnvalue_5 :!class /Script/Engine.KismetSystemLibrary) (Utilities|IsValid _returnvalue_3 :!class /Script/Engine.KismetSystemLibrary))
              (Audio|Components|Audio|FadeOut _returnvalue_5 1.0 0.0 "Linear" :!class /Script/Engine.AudioComponent))))))))

(fn OnRep_bMusicEnabled
  (Variables|/Script/Engine.PrimitiveComponent|SetbGenerateOverlapEvents true (Variables|Default|GetOverlapSphere)))

(fn OnRep_RepStruct
  (CallFunction|OnRepStructChanged (Variables|Default|GetRepStruct)))

(fn ConstructionScript
  (bind (_benablesourcedataoverride _benablesendtoaudiolink _audiolinksettingsoverride _nonspatializedradiusstart _nonspatializedradiusend _nonspatializedradiusmode _attenuationshapeextents) (Utilities|Struct|BreakSoundAttenuationSettings (Variables|/Script/Engine.SoundAttenuation|GetAttenuation (Variables|Default|GetAttenuation))))
  (Components|Sphere|SetSphereRadius (Variables|Default|GetOverlapSphere) (.x _attenuationshapeextents) true))
