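; WelcomePartyAudioRoutingComponent is a Blueprint that declares custom events for controlling audio in the Welcome Party biome sound set.
; It exposes Start, which takes an Audio Component, Stop, which takes a fade-out duration, and Reset.
; All three events have empty bodies here, so the file shows an interface-like surface with no routing logic.
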(event Custom|Start
  (params (Audio%20Component :type Audio%20Component%20Object%20Reference)))

(event Custom|Stop
  (params (FadeOutDuration :type Float%20%28double-precision%29)))

(event Custom|Reset)
