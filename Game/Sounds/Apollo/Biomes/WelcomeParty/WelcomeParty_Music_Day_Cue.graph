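# Sound cue for the daytime music set of the WelcomeParty map biome, with stage attenuation.
# A single looping wave player plays the full day set.
# Base submix output is off, so the audio is sent at full level to the WelcomeParty audio bus rather than played directly.
# A send to the music audio-analysis submix is listed, but submix sends are disabled on the cue.
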SoundCue /Game/Sounds/Apollo/Biomes/WelcomeParty/WelcomeParty_Music_Day_Cue
root=WavePlayer_0
nodes [{id=WavePlayer_0, properties {bLooping=true, soundWaveAssetPtr=/Game/Sounds/Apollo/Biomes/WelcomeParty/Music/FN_Map_WelcomeParty_FullSet_Day}}]
properties {
 attenuationSettings=&/WelcomeParty_Music_Stage_SourceBusSend_Att
 bEnableBaseSubmix=false
 bEnableSubmixSends=false
 maxDistance=2097152
 volumeMultiplier=1
 busSends [{audioBus=&/WelcomeParty_AudioBus, minSendLevel=0, sendLevel=1, soundSourceBus=null}]
 soundSubmixSends [{minSendLevel=0, sendStage=PreDistanceAttenuation, soundSubmix=&Fortnite_AudioAnalysis_Music}]
}
